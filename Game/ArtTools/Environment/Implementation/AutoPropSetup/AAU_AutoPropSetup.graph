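# Editor utility that auto-generates prop Blueprints from selected Static Meshes, for environment art setup.
# Meshes under a Props path are sized by bounding-box volume against thresholds, picking an SDP, Small, Medium or Large blockout parent class.
# MakeBPs creates a BP_-named child in a Blueprints folder, sets the mesh on its class default object and saves it.
# Only SDP results join the shared FN_BR_SimpleBuildingProps collection; Sets meshes get no processing.
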EditorUtilityBlueprint /Game/ArtTools/Environment/Implementation/AutoPropSetup/AAU_AutoPropSetup
graphs [
 AddBpToCollection {nodes [
   n1 = AddBpToCollection!() -> (then, BPReference:Blueprint Object Reference) {is=title K2Node_FunctionEntry}
   > AddAssetToCollection!(
    execute
    self:Collection Manager Subsystem Object Reference=Get CollectionManagerScriptingSubsystem()
    Collection:"Collection Structure (by ref)"=Make Collection!(Container:Name=Game, Name:Name=FN_BR_SimpleBuildingProps, ShareType:ECollectionScriptingShareType Enum=Shared) -> (CollectionScriptingRef:Collection Structure) {is=title K2Node_MakeStruct}
    AssetPath:"Soft Object Path Structure (by ref)"=GetSoftObjectPath(Object=@n1)
   ) -> (then, ReturnValue:bool)
  ]}
 AutoBlueprint_Function {nodes [
   n1 = AutoBlueprint_Function!() -> (then, Assets:Array of Object References) {is=title K2Node_FunctionEntry}
   > n2 = For Each Loop(Array:Array of Object References=@n1) -> (Array Element:Object Reference)
   > n3 = Set O_StaticMesh(O_StaticMesh:Object Reference=@n2) -> (Output_Get:Object Reference)
   > CheckIfRedirector!(execute, self, Object:Object Reference=@n3) -> (then)
   > n4 = Cast To StaticMesh(Object=Get O_StaticMesh() -> (O_StaticMesh:Object Reference))
   > n5 = Check if Prop or Set!(execute:Exec, Object:Object Reference=@n4) -> (Set:Exec, Prop:Exec) {is=title K2Node_MacroInstance}
   n6 = Cast To StaticMesh(execute=@n5.Prop, Object=Get O_StaticMesh())
   K2Node_Composite_1 = CalculateM3Volume_2!(StaticMesh:Static Mesh Object Reference=@n6) -> (Volume:double=~) {is=title K2Node_Composite}
   > n7 = Set F_StaticMeshVolume(F_StaticMeshVolume:double=@K2Node_Composite_1) -> (Output_Get:double)
   > PrintString(InString="To String (Float)"(InDouble=@n7))
   > n8 = GetSMSizeCategory!(execute, self, Volume:double=Get F_StaticMeshVolume() -> (F_StaticMeshVolume:double)) -> (then, NewParam:ENUM_PropVolumeSize Enum=NewEnumerator0)
   > n9 = Switch on ENUM_PropVolumeSize!(execute, Selection:ENUM_PropVolumeSize Enum=@n8) -> (NewEnumerator0:Exec, NewEnumerator1:Exec, NewEnumerator2:Exec, NewEnumerator3:Exec) {is=title K2Node_SwitchEnum}
   > n10 = MakeBPs!(
    execute
    self
    ParentClass:Object Class Reference=Get BlockoutPropSDPBPClass() -> (BlockoutPropSDPBPClass:Blockout Prop Small SDP Class Reference)
   ) -> (then, BPReference:Blueprint Object Reference, OutputPin:String)
   > AddBpToCollection!(execute, self, BPReference:Blueprint Object Reference=@n10) -> (then) {is=title K2Node_CallFunction}
   MakeBPs(
    execute=@n9.NewEnumerator1
    ParentClass=Get BlockoutPropSmallBPClass() -> (BlockoutPropSmallBPClass:BP Blockout Prop Small A Class Reference)
   )
   MakeBPs(
    execute=@n9.NewEnumerator2
    ParentClass=Get BlockoutPropMediumBPClass() -> (BlockoutPropMediumBPClass:BP Blockout Prop Medium A Class Reference)
   )
   MakeBPs(
    execute=@n9.NewEnumerator3
    ParentClass=Get BlockoutPropLargeBPClass() -> (BlockoutPropLargeBPClass:BP Blockout Prop Large A Class Reference)
   )
  ]}
 CalculateM3Volume_2 {
  path=AutoBlueprint_Function.K2Node_Composite_1.CalculateM3Volume_2
  nodes [
   n1 = GetBoundingBox!(self:Static Mesh Object Reference=Inputs!() -> (StaticMesh:Static Mesh Object Reference)) -> (ReturnValue_Min:Vector=~, ReturnValue_Max:Vector=~, ReturnValue_IsValid:bool=~)
   n2 = BreakVector(InVec=vector - vector(A=@n1.ReturnValue_Max, B=@n1))
   Outputs!(Volume:double=float / float(A=float * float!(A=@n2, B=@n2.Y, C:double=@n2.Z) -> (ReturnValue) {is=title K2Node_PromotableOperator}, B=1000000))
  ]
 }
 CheckIfPropOrSet {nodes [
   n1 = Inputs!() -> (execute:Exec, Object:Object Reference)
   n2 = ParseIntoArray(SourceString=Get Object Path String(Object=@n1), Delimiter=/)
   > n3 = Branch(Condition=Contains Item(TargetArray:Array of Strings=@n2, ItemToFind:"String (by ref)"=Sets))
   > Outputs!(Set:Exec, Prop:Exec=@n4.then)
   n4 = Branch(execute=@n3.else, Condition=Contains Item(TargetArray=@n2, ItemToFind=Props))
   Show Message Dialog(
    execute=@n4.else
    Title=Get WarningTitle() -> (WarningTitle:Text)
    Message=Get WarningMessage() -> (WarningMessage:Text)
    MessageType=Ok
    DefaultValue=No
    MessageCategory=Warning
   )
  ]}
 CheckIfRedirector {nodes [
   n1 = CheckIfRedirector!() -> (then, Object:Object Reference) {is=title K2Node_FunctionEntry}
   > n2 = Get Asset By Object Path!(
    execute
    self:Asset Registry Interface=GetAssetRegistry()
    ObjectPath:"Soft Object Path Structure (by ref)"=GetSoftObjectPath(Object=@n1)
    bIncludeOnlyOnDiskAssets:bool=false
    bSkipARFilteredAssets:bool=true
   ) -> (then, ReturnValue:Asset Data Structure) {is=title K2Node_CallFunction}
   > Branch(Condition=IsRedirector!(InAssetData:"Asset Data Structure (by ref)"=@n2) -> (ReturnValue:bool))
   > Return Node()
  ]}
 EnsureBlueprintsFolderExists {nodes [
   n1 = EnsureBlueprintsFolderExists!() -> (then, PackagePath:Name) {is=title K2Node_FunctionEntry}
   n2 = "To String (Name)"(InName=@n1)
   n3 = ParseIntoArray(SourceString=@n2, Delimiter=/)
   n4 = Length(TargetArray:Array of Strings=@n3)
   n5 = Replace(SourceString=@n2, From="Get (a copy)"(Array=@n3, Dimension 1=int - int(A=@n4)), To=Blueprints, SearchCase=CaseSensitive)
   > n6 = DoesDirectoryExist!(execute, DirectoryPath:String=@n5) -> (then, ReturnValue:bool)
   > n7 = Branch(Condition=@n6)
   Return Node!(execute=[@n8.then, @n7.then], BlueprintsFolderPath:String=@n5)
   n8 = MakeDirectory!(execute=@n7.else, DirectoryPath:String=@n5) -> (then, ReturnValue:bool)
  ]}
 EventGraph {nodes [PropsAutoBlueprints()]}
 GetSMSizeCategory {nodes [
   n1 = GetSMSizeCategory!() -> (then, Volume:double=~) {is=title K2Node_FunctionEntry}
   > Set F_Volume(F_Volume:double=@n1) -> (Output_Get:double)
   > n2 = Branch(Condition="float <= float"(A=Get F_Volume() -> (F_Volume:double), B=Get F_SDPVolume() -> (F_SDPVolume:double)) -> (ReturnValue=~))
   > Return Node!(execute, NewParam:ENUM_PropVolumeSize Enum=NewEnumerator0)
   n3 = Branch(execute=@n2.else, Condition="float <= float"(A=Get F_Volume(), B=Get F_SmallPropVolume() -> (F_SmallPropVolume:double)) -> (ReturnValue=~))
   > Return Node(NewParam=NewEnumerator1)
   n4 = Branch(execute=@n3.else, Condition="float <= float"(A=Get F_Volume(), B=Get F_MediumPropVolume() -> (F_MediumPropVolume:double)) -> (ReturnValue=~))
   > Return Node(NewParam=NewEnumerator2)
   Return Node(execute=@n4.else, NewParam=NewEnumerator3)
  ]}
 MakeAssetPath {nodes [
   n1 = Inputs!() -> (BlueprintsFolderPath:String, AssetName:Name)
   n2 = "To String (Name)"(InName=@n1.AssetName)
   Outputs!(AssetPath:String=Append(A=Append(A=@n1, B=/), B=Replace(SourceString=@n2, From=SM_, To=BP_, SearchCase=CaseSensitive)))
  ]}
 MakeBPs {nodes [
   n1 = MakeBPs!() -> (then, ParentClass:Object Class Reference) {is=title K2Node_FunctionEntry}
   > n2 = Get Asset By Object Path!(
    execute
    self=GetAssetRegistry()
    ObjectPath=GetSoftObjectPath(Object=Get O_StaticMesh())
    bIncludeOnlyOnDiskAssets=false
    bSkipARFilteredAssets=true
   ) -> (
    then
    ReturnValue_PackageName:Name
    ReturnValue_PackagePath:Name
    ReturnValue_AssetName:Name
    ReturnValue_AssetClassPath:Top Level Asset Path Structure
   )
   > n3 = EnsureBlueprintsFolderExists!(execute, self, PackagePath:Name=@n2.ReturnValue_PackagePath) -> (then, BlueprintsFolderPath:String) {is=title K2Node_CallFunction}
   n4 = Make Asset Path!(BlueprintsFolderPath:String=@n3, AssetName:Name=@n2.ReturnValue_AssetName) -> (AssetPath:String) {is=title K2Node_MacroInstance}
   > n5 = CreateBlueprintAssetWithParent!(execute, AssetPath:String=@n4, ParentClass:Object Class Reference=@n1) -> (then, ReturnValue:Blueprint Object Reference)
   > n6 = GeneratedClass!(execute, BlueprintObj:Blueprint Object Reference=@n5) -> (then, ReturnValue:Object Class Reference)
   > n7 = GetClassDefaultObject!(execute, Class:Object Class Reference=@n6) -> (then, ReturnValue:Object Reference)
   > SetEditorProperty(Object=@n7, PropertyName=StaticMesh, PropertyValue:Object Reference=Get O_StaticMesh())
   > SaveAsset!(execute, AssetToSave:String=@n4, bOnlyIfIsDirty:bool=true) -> (then, ReturnValue:bool)
   > Return Node!(execute, BPReference:Blueprint Object Reference=@n5, OutputPin:String=@n4)
  ]}
 PropsAutoBlueprints {nodes [
   PropsAutoBlueprints() {is=title K2Node_FunctionEntry}
   > n1 = GetSelectedAssets()
   > AutoBlueprint_Function!(execute, self, Assets:Array of Object References=@n1) -> (then) {is=title K2Node_CallFunction}
  ]}
]
variables [
 BlockoutPropLargeBPClass {type=Class, default=/Game/Environments/Blockout/Blueprints/Props/BP_Blockout_Prop_Large_A.BP_Blockout_Prop_Large_A_C}
 BlockoutPropMediumBPClass {type=Class, default=/Game/Environments/Blockout/Blueprints/Props/BP_Blockout_Prop_Medium_A.BP_Blockout_Prop_Medium_A_C}
 BlockoutPropSDPBPClass {type=Class, default=/Game/Environments/Blockout/Blueprints/Props/Blockout_Prop_Small_SDP.Blockout_Prop_Small_SDP_C}
 BlockoutPropSmallBPClass {type=Class, default=/Game/Environments/Blockout/Blueprints/Props/BP_Blockout_Prop_Small_A.BP_Blockout_Prop_Small_A_C}
 F_LargePropVolume {type=number, default=70}
 F_MediumPropVolume {type=number, default=12}
 F_SDPVolume {type=number, default=0.7}
 F_SmallPropVolume {type=number, default=0.85}
 F_StaticMeshVolume {type=number, default=9.9321}
 NewVar_0 {type=boolean}
 O_StaticMesh {type=Object, default=null}
 S_BlockoutPath {type=string, default=/Game/Environments/Blockout}
 S_SmallPropBP {type=string, default=SM_Blockout_Prop_Small_A}
 WarningMessage {type=string, default="Please ensure the Static Meshes are in a \"/Props/Meshes\" folder"}
 WarningTitle {type=string, default=Warning}
]
