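# Editor asset action utility that creates BuildingStaticGeo (BSG) versions of selected Blueprints, skipping those whose native parent derives from SimpleDestructibleProp.
# It searches referencers of each Blueprint's /Meshes/ dependencies through the Asset Registry for an existing BuildingStaticGeo asset with the BSG_ name.
# If none is found, it duplicates the Blueprint into a BSG subfolder, renames and reparents it.
# The Content Browser then selects the new or existing asset.
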EditorUtilityBlueprint /Game/ArtTools/Environment/Implementation/CreateBSGVersion/AAU_CreateBSGVersion
graphs [
 Convert_BP_To_BSG_Name {nodes [
   n1 = Convert_BP_To_BSG_Name!() -> (then, InName:Name) {is=title K2Node_FunctionEntry}
   > Set L_Name(L_Name:String="To String (Name)"(InName=@n1)) -> (Output_Get:String)
   > n2 = Branch(Condition=StartsWith(SourceString=Get L_Name() -> (L_Name:String), InPrefix=BSG_))
   > Return Node!(execute, ReturnValue:String=Get L_Name())
   n3 = Sequence(execute=@n2.else)
   Return Node(execute=@n3.then_1, ReturnValue=Get L_Name())
   > n4 = Branch(Condition=StartsWith(SourceString=Get L_Name(), InPrefix=BP_))
   > Set L_Name(L_Name=Replace(SourceString=Get L_Name(), From=BP_, To=BSG_))
   Set L_Name(execute=@n4.else, L_Name=Append(A=BSG_, B=Get L_Name()))
  ]}
 CreateBSGVersion_Function {nodes [
   n1 = CreateBSGVersion_Function!() -> (then, TargetArray:Array of Object References) {is=title K2Node_FunctionEntry}
   > n2 = Filter by Class!(
    execute
    TargetArray:Array of Object References=@n1
    ObjectClass:Object Class Reference=/Script/Engine.Blueprint
    FilterType:EEditorScriptingFilterType Enum
   ) -> (then, ReturnValue:Array of Blueprint Object References)
   > n3 = For Each Loop(Array:Array of Blueprint Object References=@n2) -> (Array Element:Blueprint Object Reference)
   n4 = CreateAssetData(InAsset=@n3, bAllowBlueprintClass=false)
   n5 = GetTagValue(InAssetData=@n4, InTagName=MakeLiteralName(Value=NativeParentClass))
   > n6 = LoadClassAsset_Blocking(AssetClass=To Soft Class Reference(SoftClassPath=MakeSoftClassPath(PathString=@n5)))
   > n7 = Branch(Condition=ClassIsChildOf(TestClass=@n6, ParentClass=/Script/FortniteGame.SimpleDestructibleProp))
   n8 = checkForExistingBSG!(execute=@n7.else, self, assetData:Asset Data Structure=@n4) -> (then, createNewBSG:bool, foundBSGs:Array of Strings)
   > n9 = Branch(Condition=@n8)
   n10 = Break Asset Data(AssetData=@n4)
   > n11 = DuplicateLoadedAsset!(
    execute
    SourceAsset:Object Reference=@n3
    DestinationAssetPath:String=Append!(
     A="To String (Name)"(InName=@n10.PackagePath)
     B=/BSG/
     C:String=Convert_BP_To_BSG_Name!(InName:Name=@n10.AssetName) -> (ReturnValue:String) {is=title K2Node_CallFunction}
    ) -> (ReturnValue)
   ) -> (then, ReturnValue:Object Reference)
   > n12 = GetBlueprintAsset(Object=@n11)
   > n13 = GetPathNameForLoadedAsset(LoadedAsset=@n12)
   > ReparentBlueprint(Blueprint=@n12, NewParentClass=/Script/FortniteGame.BuildingStaticGeo)
   > SyncBrowserToObjects(AssetPaths=Make Array("[0]":String=@n13) -> (Array:Array of Strings))
   Branch(execute=@n9.else, Condition="integer > integer"(A=Length(TargetArray:Array of Strings=@n8.foundBSGs)))
   > SyncBrowserToObjects(AssetPaths=@n8.foundBSGs)
  ]}
 Create_BSG_Version {nodes [
   Create_BSG_Version() {is=title K2Node_FunctionEntry}
   > n1 = GetSelectedAssets()
   > CreateBSGVersion_Function!(execute, self, TargetArray:Array of Object References=@n1) -> (then) {is=title K2Node_CallFunction}
  ]}
 EventGraph {links [], nodes []}
 checkForExistingBSG {nodes [
   n1 = checkForExistingBSG!() -> (then, assetData:Asset Data Structure) {is=title K2Node_FunctionEntry}
   n2 = Break Asset Data(AssetData=@n1)
   > Set l_currentPackagePath(l_currentPackagePath:Name=@n2.PackagePath) -> (Output_Get:Name=None)
   > n3 = Get Dependencies!(
    execute
    self:Asset Registry Interface=GetAssetRegistry()
    PackageName:Name=@n2
    DependencyOptions:"Asset Registry Dependency Options Structure (by ref)"=Make Asset Registry Dependency Options(bIncludeSoftPackageReferences=True)
   ) -> (then, OutDependencies:Array of Names, ReturnValue:bool) {is=title K2Node_CallFunction}
   > n4 = Branch(Condition="integer > integer"(A=Length(TargetArray:Array of Names=@n3)))
   > n5 = For Each Loop(Array:Array of Names=@n3) -> (Array Element:Name)
   n6 = Get l_foundBSGs() -> (l_foundBSGs:Array of Strings)
   n7 = Branch(execute=@n5.Completed, Condition="integer > integer"(A=Length(TargetArray:Array of Strings=@n6)))
   > Return Node!(execute, createNewBSG:bool=false, foundBSGs:Array of Strings=Get l_foundBSGs())
   Return Node(execute=@n7.else, createNewBSG=true)
   Branch(execute=@n5.LoopBody, Condition=Contains(SearchIn="To String (Name)"(InName=@n5), Substring=/Meshes/))
   > n8 = Get Referencers!(
    execute
    self:Asset Registry Interface=GetAssetRegistry()
    PackageName:Name=@n5
    ReferenceOptions:"Asset Registry Dependency Options Structure (by ref)"=Make Asset Registry Dependency Options(bIncludeSoftPackageReferences=True)
   ) -> (then, OutReferencers:Array of Names, ReturnValue:bool) {is=title K2Node_CallFunction}
   > n9 = For Each Loop(Array=@n8)
   > Set l_currentReferencer(l_currentReferencer:Name=@n9) -> (Output_Get:Name=None)
   > n10 = Branch(Condition=Contains Item(TargetArray=Get l_testedReferencers() -> (l_testedReferencers:Array of Names), ItemToFind=@n9))
   Add Unique(
    execute=@n10.else
    TargetArray:Array of Names=Get l_testedReferencers()
    NewItem:"Name (by ref)"=Get l_currentReferencer() -> (l_currentReferencer:Name=None)
   )
   n11 = Convert_BP_To_BSG_Name(InName=Break Asset Data(AssetData=Get assetData() -> (assetData:Asset Data Structure)).AssetName)
   n12 = "To String (Name)"(InName=Get l_currentReferencer() -> (l_currentReferencer=None))
   n13 = Get l_currentReferencer() -> (l_currentReferencer=None)
   n14 = Get l_currentPackagePath() -> (l_currentPackagePath:Name=None)
   > Branch(Condition=AND Boolean(A=Contains(SearchIn=@n12, Substring=@n11), B="Not Equal (Name)"(A=@n14, B=@n13) -> (ReturnValue=~)))
   > n15 = LoadBlueprintClass!(execute, AssetPath:String="To String (Name)"(InName=Get l_currentReferencer() -> (l_currentReferencer=None))) -> (then, ReturnValue:Object Class Reference)
   > Branch(Condition=ClassIsChildOf(TestClass=@n15, ParentClass=&BuildingStaticGeo))
   > Add(TargetArray=Get l_foundBSGs(), NewItem="To String (Name)"(InName=Get l_currentReferencer() -> (l_currentReferencer=None))) {is=title K2Node_CallArrayFunction}
   Return Node(execute=@n4.else, createNewBSG=false)
  ]}
]
variables []
