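# Editor utility that reduces Nanite vertex counts on selected static meshes by raising their Trim Relative Error.
# OptimizeNaniteTrimError passes the selection to ReduceNanite, which checks out assets and sets a limit of ten times LOD0 vertices, updated only when below 200,000.
# A while loop reapplies Nanite settings with stepped-up trim error, re-reading counts through GetVertCounts until within the limit.
# Processed meshes accumulate in an array that is saved afterwards.
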EditorUtilityBlueprint /Game/ArtTools/Environment/Implementation/ReduceNanite/ReduceNanite
graphs [
 EventGraph {links [], nodes []}
 GetVertCounts {nodes [
   GetVertCounts() {is=title K2Node_FunctionEntry}
   > n1 = Sequence()
   n2 = Get Current Mesh() -> (Current Mesh:Object Reference)
   > n3 = Cast To StaticMesh(Object=@n2)
   > Set LOD0Verts(LOD0Verts:int=GetNumVertices!(self:Static Mesh Object Reference=@n3, LODIndex:int=0) -> (ReturnValue:int=0)) -> (Output_Get:int=0)
   n4 = Cast To StaticMesh(execute=@n1.then_1, Object=@n2)
   > Set NaniteVerts(NaniteVerts:int=Get Num Nanite Vertices!(StaticMesh:Static Mesh Object Reference=@n4) -> (ReturnValue:int=0) {is=title K2Node_CallFunction}) -> (Output_Get:int=0)
  ]}
 OptimizeNaniteTrimError {nodes [
   OptimizeNaniteTrimError() {is=title K2Node_FunctionEntry}
   > n1 = GetSelectedAssets()
   > ReduceNanite!(execute, self, Assets:Array of Object References=@n1) -> (then)
  ]}
 ReduceNanite {nodes [
   n1 = ReduceNanite!() -> (then, Assets:Array of Object References) {is=title K2Node_FunctionEntry}
   > n2 = CheckoutLoadedAssets(AssetsToCheckout=@n1)
   > Branch(Condition=@n2)
   > n3 = For Each Loop(Array:Array of Object References=@n1) -> (Array Element:Object Reference)
   SaveLoadedAssets(execute=@n3.Completed, AssetsToSave=Get ARAssetsToSave() -> (ARAssetsToSave:Array of Object References), bOnlyIfIsDirty=true)
   Set Current Mesh(execute=@n3.LoopBody, Current Mesh:Object Reference=@n3) -> (Output_Get:Object Reference)
   > n4 = Sequence!(execute) -> (then_0, then_1, then_2:Exec)
   > GetVertCounts() {is=title K2Node_CallFunction}
   Set LOD0Vert10x(execute=@n4.then_1, LOD0Vert10x:int=int * int(A=Get LOD0Verts() -> (LOD0Verts:int=0), B=10)) -> (Output_Get:int=0)
   n5 = Get LOD0Vert10x() -> (LOD0Vert10x:int=0)
   > Branch(Condition="integer < integer"(A=@n5, B=200000))
   > Set NaniteVertLimit(NaniteVertLimit:int=@n5) -> (Output_Get:int=0)
   n6 = While Loop!(
    execute:Exec=@n4.then_2
    Condition:bool="integer > integer"(A=Get NaniteVerts() -> (NaniteVerts:int=0), B=Get NaniteVertLimit() -> (NaniteVertLimit:int=0))
   ) -> (LoopBody:Exec, Completed:Exec) {is=title K2Node_MacroInstance}
   Set TrimMultiply(execute=@n6.Completed, TrimMultiply:double=1) -> (Output_Get:double)
   > Add(TargetArray:Array of Object References=Get ARAssetsToSave(), NewItem:"Object Reference (by ref)"=Get Current Mesh()) {is=title K2Node_CallArrayFunction}
   n7 = Cast To StaticMesh(execute=@n6.LoopBody, Object=Get Current Mesh())
   > n8 = Cast To StaticMesh(Object=Get Current Mesh())
   n9 = Break Mesh Nanite Settings!(MeshNaniteSettings:"Mesh Nanite Settings Structure (by ref)"=GetNaniteSettings!(self:Static Mesh Editor Subsystem Object Reference=Get StaticMeshEditorSubsystem(), StaticMesh:Static Mesh Object Reference=@n7) -> (ReturnValue:Mesh Nanite Settings Structure)) -> (
    bEnabled:bool=~
    bExplicitTangents:bool=~
    bLerpUVs:bool=~
    bSeparable:bool=~
    bVoxelNDF:bool=~
    bVoxelOpacity:bool=~
    ShapePreservation:ENaniteShapePreservation Enum
    PositionPrecision:int
    NormalPrecision:int
    TangentPrecision:int
    BoneWeightPrecision:int
    KeepPercentTriangles:float=~
    TrimRelativeError:float=~
    GenerateFallback:ENaniteGenerateFallback Enum
    FallbackTarget:ENaniteFallbackTarget Enum
    FallbackPercentTriangles:float=~
    FallbackRelativeError:float=~
    MaxEdgeLengthFactor:float=~
    NumRays:int
    VoxelLevel:int
    RayBackUp:float=~
    DisplacementUVChannel:int
   ) {is=title K2Node_BreakStruct}
   n10 = Get TrimAmount() -> (TrimAmount:double)
   n11 = Get TrimMultiply() -> (TrimMultiply:double)
   > SetNaniteSettings!(
    execute
    self:Static Mesh Editor Subsystem Object Reference=Get StaticMeshEditorSubsystem()
    StaticMesh:Static Mesh Object Reference=@n8
    NaniteSettings:Mesh Nanite Settings Structure=Make Mesh Nanite Settings!(
     bEnabled:bool=@n9
     bExplicitTangents:bool=@n9.bExplicitTangents
     bLerpUVs:bool=@n9.bLerpUVs
     bSeparable:bool=@n9.bSeparable
     bVoxelNDF:bool=@n9.bVoxelNDF
     bVoxelOpacity:bool=@n9.bVoxelOpacity
     ShapePreservation:ENaniteShapePreservation Enum=@n9.ShapePreservation
     PositionPrecision:int=@n9.PositionPrecision
     NormalPrecision:int=@n9.NormalPrecision
     TangentPrecision:int=@n9.TangentPrecision
     BoneWeightPrecision:int=@n9.BoneWeightPrecision
     KeepPercentTriangles:float=@n9.KeepPercentTriangles
     TrimRelativeError:float=float * float(A=@n10, B=@n11) {is=title K2Node_PromotableOperator}
     GenerateFallback:ENaniteGenerateFallback Enum=@n9.GenerateFallback
     FallbackTarget:ENaniteFallbackTarget Enum=@n9.FallbackTarget
     FallbackPercentTriangles:float=@n9.FallbackPercentTriangles
     FallbackRelativeError:float=@n9.FallbackRelativeError
     MaxEdgeLengthFactor:float=@n9.MaxEdgeLengthFactor
     NumRays:int=@n9.NumRays
     VoxelLevel:int=@n9.VoxelLevel
     RayBackUp:float=@n9.RayBackUp
     DisplacementUVChannel:int=@n9.DisplacementUVChannel
    ) -> (MeshNaniteSettings:Mesh Nanite Settings Structure) {is=title K2Node_MakeStruct}
    bApplyChanges:bool=true
   ) -> (then)
   > GetVertCounts()
   > Branch(Condition="integer > integer"(A=Get NaniteVerts() -> (NaniteVerts=0), B=Get NaniteVertLimit() -> (NaniteVertLimit=0)))
   > Set TrimMultiply(TrimMultiply=float + float(A=Get TrimMultiply(), B=1))
  ]}
]
variables [
 ARAssetsToSave {
  type="array<Object>"
  default [
   /Game/Environments/Asteria/Terrain/Frozen/Meshes/SM_Asteria_IceChunk_B
   /Game/Environments/Asteria/Terrain/Riverbank/Meshes/SM_Asteria_Riverbank_Corner_180_A
   /Game/Environments/Asteria/Terrain/Riverbank/Meshes/SM_Asteria_Riverbank_Straight_A
   /Game/Environments/Helios/Terrain/Cliffs/Meshes/SM_Helios_MtnTop_A
   /Game/Environments/Asteria/Sets/LightCastle/Meshes/SM_LightCastle_Cliff_C
   /Game/Environments/Asteria/Sets/Quarry/Meshes/SM_Quarry_Bridge_2x1x1
   /Game/Environments/Asteria/Sets/Quarry/Meshes/SM_Quarry_Cube_05x1x1_A
   /Game/Environments/Asteria/Sets/Quarry/Meshes/SM_Quarry_Cube_2x2x2_D
   /Game/Environments/Helios/Rocks/SpikeyRocks/Meshes/SM_SpikeyRamp_Indestructible_C
   /Game/Environments/Helios/Rocks/SpikeyRocks/Meshes/SM_SpikeyRamp_Indestructible_C_B
   /Game/Environments/Helios/Rocks/SpikeyRocks/Meshes/SM_SpikeyRamp_Indestructible_C_Collision
   /Hera/Props/BeefJerky/BeefJerky_BearStatue_B/Meshes/SM_BeefJerky_BearStatue_B
  ]
 }
 Current Mesh {type=Object, default=&SM_BeefJerky_BearStatue_B}
 LOD0Vert10x {type=integer, default=79080}
 LOD0Verts {type=integer, default=7908}
 NaniteVertLimit {type=integer, default=79080}
 NaniteVerts {type=integer, default=30646}
 TrimAdditional {type=number}
 TrimAmount {type=number, default=0.05}
 TrimMultiply {type=number, default=1}
]
